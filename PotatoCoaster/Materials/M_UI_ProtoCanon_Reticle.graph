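# Translucent UI material for the ProtoCanon reticle, with idle and charging looks and a ChargeAmount parameter.
# One distance-field texture supplies idle shape, charge bar and brackets by channel, cut into antialiased fills and strokes.
# Left and right halves get separate offset UVs, masks and angle-gradient bar fills.
# Three-color gradients, smoothstep thresholds and time-driven pulses build charge color and glow, with scalar parameters tuning scale, thickness, offset, softness and opacity.
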Material /PotatoCoaster/Materials/M_UI_ProtoCanon_Reticle
graph {
 properties {bAutomaticallySetUsageInEditor=false, blendMode=BLEND_Translucent, materialDomain=MD_UI, usedShadingModels=Default Lit}
 nodes [
  Constant3Vector() {constant=<rgba 0.491 0.0252 0.0103 1>}
  Constant3Vector() {constant=<rgba 0.4179 0.004 0.0284 1>}
  Divide() {constB=1}
  LinearInterpolate(A=Constant() {r=0.19}, B=Constant() {r=0.21}, Alpha=OneMinus(None=NamedRerouteUsage()))
  n1 = LinearInterpolate(A=Constant() {r=1}, B=NamedRerouteUsage(), Alpha=NamedRerouteUsage())
  n2 = Add(A=NamedRerouteUsage(), B=NamedRerouteUsage())
  LinearInterpolate(A=NamedRerouteUsage(), B=Subtract(A=Saturate(None=@n1), B=OneMinus(None=@n2)), Alpha=Ceil(None=NamedRerouteUsage()))
  LinearInterpolate(A=Constant() {r=0.135}, B=Constant() {r=0.145}, Alpha=OneMinus(None=NamedRerouteUsage()))
  LinearInterpolate(A=NamedRerouteUsage(), B=NamedRerouteUsage(), Alpha=Ceil(None=NamedRerouteUsage()))
  LinearInterpolate(A=Constant() {r=0.14}, B=Constant() {r=0.14}, Alpha=OneMinus(None=NamedRerouteUsage()))
  n3 = Constant() {r=0.35}
  n4 = Saturate(None=MaterialFunctionCall(
    UVs=MaterialFunctionCall(Texture Scale=@n3) {materialFunction=&ScaleUVsByCenter}.UVs
    RotationAngle=Constant() {r=0.75}
   ) {materialFunction=&MF_UI_LinearGradient}.Result)
  LinearInterpolate(
   A=MaterialFunctionCall(
    A=Constant3Vector() {constant=<rgba 1 0.2151 0 1>}
    B=Constant3Vector() {constant=<rgba 1 0 0 0>}
    C=Constant3Vector() {constant=<rgba 0.305 0 1 0>}
    Alpha=@n4
   ) {materialFunction=&Lerp_3Color}.Result
   B=MaterialFunctionCall(
    A=Constant3Vector() {constant=<rgba 1 0.2151 0 1>}
    B=Constant3Vector() {constant=<rgba 1 0.1562 0.1562 0>}
    C=Constant3Vector() {constant=<rgba 1 0.7491 0.1355 1>}
    Alpha=@n4
   ) {materialFunction=&Lerp_3Color}.Result
   Alpha=NamedRerouteUsage()
  ) {constA=1}
  LinearInterpolate(A=Constant() {r=0.19}, B=Constant() {r=0.21}, Alpha=OneMinus(None=NamedRerouteUsage()))
  MaterialFunctionCall(Texture Scale=Constant() {r=0.35}) {materialFunction=&ScaleUVsByCenter}
  n5 = Max(A=NamedRerouteUsage(), B=NamedRerouteUsage())
  Multiply(A=NamedRerouteUsage(), B=@n5)
  Multiply(A=Time() {bOverride_Period=true, period=4}) {constB=0.25}
  n6 = Multiply(A=NamedRerouteUsage(), B=NamedRerouteUsage())
  n7 = Multiply(A=NamedRerouteUsage(), B=NamedRerouteUsage())
  n8 = MaterialFunctionCall(DistanceField=Saturate(None=Max(A=@n6, B=@n7)), Threshold=Constant() {r=0.2}, Smoothing=Constant() {r=0.25}) {materialFunction=&MF_UI_AntiAliasing}
  n9 = TextureCoordinate() {vTiling=1.4}
  n10 = Multiply(A=MaterialFunctionCall(UV Channel=OneMinus(None=@n9)) {materialFunction=&LinearGradient}.VGradient)
  n11 = LinearInterpolate(
   B=Multiply(A=ScalarParameter() {defaultValue=0.5, parameterName=EffectOpacity}, B=Multiply(A=@n8.Outer, B=@n10))
   Alpha=SmoothStep(Value=NamedRerouteUsage()) {constMax=0.95, constMin=0.5}
  )
  n12 = Constant() {r=0.15}
  n13 = Fmod(A=Time() {bOverride_Period=true, period=1}, B=@n12)
  Multiply(A=@n11, B=Round(None=Divide(A=@n13, B=@n12)))
  n14 = NamedRerouteUsage()
  Multiply(A=@n14, B=LinearInterpolate(Alpha=NamedRerouteUsage()) {constA=1.14, constAlpha=0, constB=1.24})
  NamedRerouteDeclaration(None=ScalarParameter() {parameterName=ChargeAmount, sortPriority=1}) {name=ChargeAmount, nodeColor=<rgba 0.3828 1 0 1>}
  n15 = TextureSample(UVs=NamedRerouteUsage()) {texture=/PotatoCoaster/Textures/T_UI_ProtoCanon_Reticle}
  n16 = MaterialFunctionCall(DistanceField=@n15.G, Threshold=NamedRerouteUsage(), Smoothing=NamedRerouteUsage()) {materialFunction=&MF_UI_AntiAliasing}
  n17 = NamedRerouteUsage()
  n18 = MaterialFunctionCall(DistanceField=@n15.B, Threshold=NamedRerouteUsage(), Smoothing=NamedRerouteUsage()) {materialFunction=&MF_UI_AntiAliasing}
  n19 = NamedRerouteUsage()
  NamedRerouteDeclaration(None=Max(A=Multiply(A=@n16.Outer, B=@n17) {desc=CHARGE BAR}, B=Multiply(A=@n18.Outer, B=@n19) {desc=BRACKETS})) {name=Bar_Mask_L, nodeColor=<rgba 1 0 0.75 1>}
  n20 = LinearInterpolate(Alpha=NamedRerouteUsage()) {constA=0.3, constAlpha=0, constB=0.51}
  n21 = MaterialFunctionCall!(UV=NamedRerouteUsage(), Rotate=Constant() {r=0.348}) {materialFunction=/AudioWidgets/AudioMaterialSlate/Functions/MF_AngleGradient}
  NamedRerouteDeclaration(None=OneMinus(None=Step(Y=@n20, X=@n21.Result) {constY=0.3})) {name=BarFill_R, nodeColor=<rgba 0 1 0.9297 1>}
  n22 = LinearInterpolate(Alpha=NamedRerouteUsage()) {constA=0.3, constAlpha=0, constB=0.51}
  n23 = OneMinus(None=MaterialFunctionCall(UV=NamedRerouteUsage(), Rotate=Constant() {r=0.652}) {materialFunction=&MF_AngleGradient}.Result)
  NamedRerouteDeclaration(None=OneMinus(None=Step(Y=@n22, X=@n23) {constY=0.3})) {name=BarFill_L, nodeColor=<rgba 1 0.6094 0 1>}
  n24 = TextureCoordinate()
  n25 = Add(A=ComponentMask(None=@n24) {r=true}) {constB=0}
  n26 = Add(A=ComponentMask(None=@n24) {g=true}) {constB=-0.065}
  NamedRerouteDeclaration(None=MaterialFunctionCall(UVs=AppendVector(A=@n25, B=@n26), Texture Scale=NamedRerouteUsage()) {materialFunction=&ScaleUVsByCenter}.UVs) {name=UV_Texture, nodeColor=<rgba 0.1953 1 0 1>}
  n27 = Saturate(None=Max(A=NamedRerouteUsage(), B=NamedRerouteUsage()))
  n28 = NamedRerouteUsage()
  n29 = Multiply(A=NamedRerouteUsage(), B=@n5)
  n30 = Saturate(None=Max(A=Multiply(A=@n27, B=@n28), B=@n29))
  n31 = Max(A=NamedRerouteUsage(), B=NamedRerouteUsage())
  n32 = NamedRerouteUsage()
  n33 = Multiply(A=NamedRerouteUsage(), B=NamedRerouteDeclaration(None=Subtract(A=@n31, B=@n32)) {name=Bars_OnlyStroke, nodeColor=<rgba 0 0.4375 1 1>}) {constB=0.1}
  NamedRerouteDeclaration(None=Max(A=Max(A=@n30, B=@n33), B=NamedRerouteUsage())) {name=ChargingState, nodeColor=<rgba 0.4766 0 1 1>}
  NamedRerouteDeclaration(None=ScalarParameter() {defaultValue=0.9, parameterName=OverallScale, sortPriority=2}) {name=Scale, nodeColor=<rgba 1 0.6797 0 1>}
  n34 = MaterialFunctionCall(DistanceField=Max(A=NamedRerouteUsage(), B=NamedRerouteUsage()), Threshold=Constant() {r=0.0005}, Smoothing=Constant() {r=0.003}) {materialFunction=&MF_UI_AntiAliasing}
  n35 = SmoothStep(Value=NamedRerouteUsage()) {constMax=0.95, constMin=0.8}
  n36 = Multiply(A=LinearInterpolate(B=Constant3Vector() {constant=<rgba 1 0.0534 0.1265 1>}, Alpha=Floor(None=@n35)) {constA=1}) {constB=0.8}
  n37 = Multiply(A=Time() {bOverride_Period=true, period=0.5}) {constB=2}
  n38 = Multiply(A=Frac(None=Sine(None=@n37))) {constB=0.2}
  n39 = Floor(None=SmoothStep(Value=NamedRerouteUsage()) {constMax=0.91, constMin=0.9})
  n40 = NamedRerouteUsage()
  n41 = Saturate(None=MaterialFunctionCall(UVs=Add(A=@n40) {constB=0}, GradientStart=Constant() {r=0.1}, GradientEnd=Constant() {r=0.45}, RotationAngle=Constant() {r=0.75}) {materialFunction=&MF_UI_LinearGradient}.Result)
  n42 = MaterialFunctionCall(
   A=Constant3Vector() {constant=<rgba 1 0 0.059 1>}
   B=Constant3Vector() {constant=<rgba 0.7305 0.0242 0.0976 1>}
   C=Constant3Vector() {constant=<rgba 0 0.2374 0.9323 1>}
   Alpha=@n41
  ) {materialFunction=&Lerp_3Color}
  n43 = Constant3Vector() {constant=<rgba 1 0.0309 0 1>}
  n44 = MaterialFunctionCall(A=@n43, B=@n43, C=Constant3Vector() {constant=<rgba 1 0 0.059 1>}, Alpha=@n41) {materialFunction=&Lerp_3Color}
  n45 = NamedRerouteUsage()
  n46 = NamedRerouteUsage()
  n47 = MaterialFunctionCall(DistanceField=@n46, Threshold=Constant() {r=0.05}, Smoothing=Constant() {r=0.3}) {materialFunction=&MF_UI_AntiAliasing}
  n48 = SmoothStep(Value=NamedRerouteUsage()) {constMin=0.5}
  n49 = Constant3Vector() {constant=<rgba 1 0.1094 0.1661 1>}
  n50 = Add(
   A=Add(A=LinearInterpolate(A=@n42.Result, B=@n44.Result, Alpha=@n45) {constA=1}, B=LinearInterpolate(B=@n38, Alpha=@n39))
   B=Multiply(A=Multiply(A=@n46, B=@n47.Outer), B=SmoothStep(Max=@n49, Value=@n48) {constMax=0.91, constMin=1}) {constB=0.7}
  )
  n51 = NamedRerouteUsage()
  NamedRerouteDeclaration(None=Add(A=Multiply(A=@n34.Inner, B=@n36) {constB=0.9}, B=Add(A=@n50, B=@n51))) {name=Color_Charging, nodeColor=<rgba 0 1 0.625 1>}
  n52 = TextureSample(UVs=NamedRerouteUsage()) {texture=&T_UI_ProtoCanon_Reticle}
  NamedRerouteDeclaration(None=Max(A=@n52.G, B=@n52.B)) {name=EffectGlowRight, nodeColor=<rgba 1 0.3047 0 1>}
  NamedRerouteDeclaration(None=@n27) {name=Bars_Mask, nodeColor=<rgba 0 0 0 1>}
  NamedRerouteDeclaration(None=Max(A=@n15.G, B=@n15.B)) {name=EffectGlowLeft, nodeColor=<rgba 0 0.0156 1 1>}
  n53 = NamedRerouteUsage()
  n54 = Constant2Vector() {g=0.65, r=2}
  NamedRerouteDeclaration(None=Multiply(A=@n11, B=MaterialFunctionCall(UVs=@n53, Size=@n54) {materialFunction=&MF_UI_Box}.Fill)) {name=GlowEffect, nodeColor=<rgba 1 0.9375 0 1>}
  NamedRerouteDeclaration(None=ScalarParameter() {defaultValue=0.03, parameterName=OutwardsOffsetAmount, sortPriority=4}) {name=Ofset, nodeColor=<rgba 0.7344 0 1 1>}
  n55 = TextureSample(UVs=NamedRerouteUsage()) {texture=&T_UI_ProtoCanon_Reticle}
  n56 = NamedRerouteUsage()
  n57 = MaterialFunctionCall(DistanceField=@n55.G, Threshold=NamedRerouteUsage(), Smoothing=@n56) {materialFunction=&MF_UI_AntiAliasing}
  n58 = MaterialFunctionCall(DistanceField=@n55.B, Threshold=NamedRerouteUsage(), Smoothing=@n56) {materialFunction=&MF_UI_AntiAliasing}
  NamedRerouteDeclaration(None=Multiply(A=Max(A=@n57.Outer, B=@n58.Outer), B=NamedRerouteUsage())) {name=BarStroke_R, nodeColor=<rgba 1 0 0.8906 1>}
  n59 = TextureSample(UVs=NamedRerouteUsage()) {texture=&T_UI_ProtoCanon_Reticle}
  n60 = NamedRerouteUsage()
  n61 = MaterialFunctionCall(DistanceField=@n59.G, Threshold=NamedRerouteUsage(), Smoothing=@n60) {materialFunction=&MF_UI_AntiAliasing}
  n62 = MaterialFunctionCall(DistanceField=@n59.B, Threshold=NamedRerouteUsage(), Smoothing=@n60) {materialFunction=&MF_UI_AntiAliasing}
  NamedRerouteDeclaration(None=Multiply(A=Max(A=@n61.Outer, B=@n62.Outer), B=NamedRerouteUsage())) {name=BarStroke_L, nodeColor=<rgba 1 0 0.1172 1>}
  NamedRerouteDeclaration(None=ScalarParameter() {defaultValue=0.25, parameterName=BarOpacity}) {name=BarOpacity, nodeColor=<rgba 0.7578 1 0 1>}
  NamedRerouteDeclaration(None=ScalarParameter() {defaultValue=0.85, parameterName=StrokeOpacity}) {name=StrokeOpacity, nodeColor=<rgba 0.9219 0 1 1>}
  NamedRerouteDeclaration(None=@n29) {name=Bars_OnlyFill, nodeColor=<rgba 0 1 0.2734 1>}
  n63 = TextureSample(UVs=NamedRerouteUsage()) {texture=&T_UI_ProtoCanon_Reticle}
  n64 = NamedRerouteDeclaration(None=MaterialFunctionCall(DistanceField=@n63.R, Threshold=NamedRerouteUsage(), Smoothing=NamedRerouteUsage()) {materialFunction=&MF_UI_AntiAliasing}.Outer) {name=IdleColor, nodeColor=<rgba 0 0 0 1>}
  n65 = NamedRerouteUsage()
  n66 = Saturate(None=MaterialFunctionCall(DistanceField=@n63.R, Threshold=NamedRerouteUsage(), Smoothing=NamedRerouteUsage()) {materialFunction=&MF_UI_AntiAliasing}.Outer)
  n67 = NamedRerouteUsage()
  NamedRerouteDeclaration(None=Max(A=Multiply(A=@n64, B=@n65) {constB=0.85}, B=Multiply(A=@n66, B=@n67))) {name=IdleOpacity, nodeColor=<rgba 0.125 1 0 1>}
  NamedRerouteDeclaration(None=@n5) {name=BarsFillMask, nodeColor=<rgba 1 0 0.0469 1>}
  NamedRerouteDeclaration(None=ScalarParameter() {defaultValue=0.02, parameterName=StrokeSoftness}) {name=StrokeSoftness, nodeColor=<rgba 0 0.6953 1 1>}
  NamedRerouteDeclaration(None=@n8.Inner) {name=BarGlow, nodeColor=<rgba 1 0 0.8906 1>}
  NamedRerouteDeclaration(None=Multiply(A=NamedRerouteUsage(), B=SmoothStep(Min=@n23, Value=@n22))) {name=ChaseBarLeft, nodeColor=<rgba 1 0 0.5156 1>}
  NamedRerouteDeclaration(None=Multiply(A=NamedRerouteUsage(), B=SmoothStep(Min=@n21.Result, Value=@n20) {constMin=0.5})) {name=ChaseBarRight, nodeColor=<rgba 0 0.8359 1 1>}
  NamedRerouteDeclaration(None=ScalarParameter() {defaultValue=0.2025, parameterName=Thinckness, sortPriority=3}) {name=Thickness, nodeColor=<rgba 1 0.3516 0 1>}
  NamedRerouteDeclaration(None=MaterialFunctionCall(UVs=NamedRerouteUsage(), Texture Scale=@n14) {materialFunction=&ScaleUVsByCenter}.UVs) {name=UV_Scale, nodeColor=<rgba 0 0.8125 1 1>}
  n68 = TextureCoordinate()
  n69 = ComponentMask(None=@n68) {r=true}
  n70 = ComponentMask(None=@n68) {g=true}
  NamedRerouteDeclaration(None=AppendVector(A=Add(A=@n69) {constB=0}, B=Add(A=@n70) {constB=0.04})) {name=UV00, nodeColor=<rgba 0.2422 0 1 1>}
  n71 = MaterialFunctionCall(Size=Constant2Vector() {g=1, r=0.33}) {materialFunction=&MF_UI_Box}
  NamedRerouteDeclaration(None=Multiply(A=NamedRerouteUsage(), B=@n71.Fill)) {name=bulletPointColor, nodeColor=<rgba 1 0 0.8906 1>}
  NamedRerouteDeclaration(None=Multiply(A=@n71.Fill, B=NamedRerouteUsage())) {name=bulletPointOpacity, nodeColor=<rgba 0 1 0.7891 1>}
  NamedRerouteDeclaration(None=ScalarParameter() {defaultValue=0.015, parameterName=OverallSmoothness}) {name=OverallSmoothness, nodeColor=<rgba 1 0 0.2578 1>}
  NamedRerouteDeclaration(None=ScalarParameter() {defaultValue=0.8, parameterName=OverallOpacity}) {name=OverallOpacity, nodeColor=<rgba 0 1 0.4609 1>}
  NamedRerouteDeclaration(None=Subtract(A=NamedRerouteUsage(), B=ScalarParameter() {defaultValue=0.035, parameterName=StrokeThickness}) {constB=0.045}) {name=StrokeThickness, nodeColor=<rgba 0.0312 0 1 1>}
  n72 = Round(None=ComponentMask(None=NamedRerouteUsage()) {r=true})
  NamedRerouteDeclaration(None=@n72) {name=MaskRightSide, nodeColor=<rgba 1 0.8672 0 1>}
  NamedRerouteDeclaration(None=OneMinus(None=@n72)) {name=MaskLeftSide, nodeColor=<rgba 0.5469 0 1 1>}
  n73 = NamedRerouteUsage()
  n74 = ComponentMask(None=@n73) {r=true}
  n75 = LinearInterpolate(B=Multiply(A=NamedRerouteUsage()) {constB=-1}, Alpha=NamedRerouteUsage()) {constAlpha=1, constB=-0.2}
  NamedRerouteDeclaration(None=AppendVector(A=Add(A=@n74, B=@n75) {constB=0.2}, B=ComponentMask(None=@n73) {g=true})) {name=UVMoveRight, nodeColor=<rgba 1 0 0.2578 1>}
  n76 = NamedRerouteUsage()
  n77 = ComponentMask(None=@n76) {r=true}
  n78 = LinearInterpolate(B=NamedRerouteUsage(), Alpha=NamedRerouteUsage()) {constAlpha=1, constB=0.2}
  NamedRerouteDeclaration(None=AppendVector(A=Add(A=@n77, B=@n78) {constB=0.2}, B=ComponentMask(None=@n76) {g=true})) {name=UVMoveLeft, nodeColor=<rgba 0 0.5781 1 1>}
  n79 = MaterialFunctionCall(DistanceField=@n52.G, Threshold=NamedRerouteUsage(), Smoothing=NamedRerouteUsage()) {materialFunction=&MF_UI_AntiAliasing}
  n80 = NamedRerouteUsage()
  n81 = MaterialFunctionCall(DistanceField=@n52.B, Threshold=NamedRerouteUsage(), Smoothing=NamedRerouteUsage()) {materialFunction=&MF_UI_AntiAliasing}
  n82 = NamedRerouteUsage()
  NamedRerouteDeclaration(None=Max(A=Multiply(A=@n79.Outer, B=@n80) {desc=CHARGE BAR}, B=Multiply(A=@n81.Outer, B=@n82) {desc=Brackets})) {name=Bar_Mask_R, nodeColor=<rgba 0.4297 1 0 1>}
  NamedRerouteUsage()
  NamedRerouteUsage()
  NamedRerouteUsage()
  NamedRerouteUsage()
  NamedRerouteUsage()
  NamedRerouteUsage()
  NamedRerouteUsage()
  NamedRerouteUsage()
  Reroute()
 ]
}
